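# Physics asset for the medium male Spy player body, covering dangling accessories: left and right guns, a right leg gun and a GPS unit.
# Each accessory body is constrained to a small base capsule with its own angular limits, tighter for the leg gun and GPS.
# Gun bases link to spine_03 in a free-rotating spine capsule chain, while the GPS base, leg-gun base and left thigh capsule have no constraint to the spine.
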PhysicsAsset /Game/Characters/Player/Male/Medium/Bodies/M_MED_Spy/Meshes/M_MED_Spy_Physics
bodies {
 dyn_gps [{shapeType=Box, extentX=13, extentY=3, extentZ=6.5, center=<2.6655 1.2785 0>}]
 dyn_gps_base [{shapeType=Capsule, length=1, radius=1}]
 dyn_gun_base_l [{shapeType=Capsule, length=1, radius=1}]
 dyn_gun_base_r [{shapeType=Capsule, length=1, radius=1}]
 dyn_gun_l [{shapeType=Box, extentX=4.0233, extentY=13.6543, extentZ=25.0369, center=<1.8533 1.1934 -5.1699>, rotation=<rot 0 -179.9213 -88.6979>}]
 dyn_gun_r [{shapeType=Box, extentX=4.0233, extentY=13.6543, extentZ=25.0369, center=<-1.8534 -1.1934 5.17>, rotation=<rot 0 -179.9213 -88.6979>}]
 dyn_leg_gun_base_r [{shapeType=Capsule, length=1, radius=1}]
 dyn_leg_gun_r [{shapeType=Capsule, length=7.25, radius=3.25, center=<1.3414 0 0>, rotation=<rot 90 0 0>}]
 spine_01 [{shapeType=Capsule, length=8.25, radius=9.5}]
 spine_02 [{shapeType=Capsule, length=7.25, radius=8.75, center=<-0.0001 -0.012 0>}]
 spine_03 [{shapeType=Capsule, length=6.25, radius=9.75, center=<0.0001 0.647 0>}]
 thigh_l [{shapeType=Capsule, length=22.25, radius=7.1, center=<-12.3539 0 0.2543>, rotation=<rot 90 0 0>}]
}
constraints [
 {bone1Name=dyn_gun_l, bone2Name=dyn_gun_base_l, swing1LimitDegrees=45, swing2LimitDegrees=45, twistLimitDegrees=45}
 {bone1Name=dyn_gun_r, bone2Name=dyn_gun_base_r, swing1LimitDegrees=45, swing2LimitDegrees=45, twistLimitDegrees=45}
 {
  bone1Name=dyn_gun_base_l
  bone2Name=spine_03
  swing1LimitDegrees=45
  swing1Motion=Free
  swing2LimitDegrees=45
  swing2Motion=Free
  twistLimitDegrees=45
  twistMotion=Free
 }
 {
  bone1Name=dyn_gun_base_r
  bone2Name=spine_03
  swing1LimitDegrees=45
  swing1Motion=Free
  swing2LimitDegrees=45
  swing2Motion=Free
  twistLimitDegrees=45
  twistMotion=Free
 }
 {
  bone1Name=spine_03
  bone2Name=spine_02
  swing1LimitDegrees=45
  swing1Motion=Free
  swing2LimitDegrees=45
  swing2Motion=Free
  twistLimitDegrees=45
  twistMotion=Free
 }
 {
  bone1Name=spine_02
  bone2Name=spine_01
  swing1LimitDegrees=45
  swing1Motion=Free
  swing2LimitDegrees=45
  swing2Motion=Free
  twistLimitDegrees=45
  twistMotion=Free
 }
 {bone1Name=dyn_leg_gun_r, bone2Name=dyn_leg_gun_base_r, swing1LimitDegrees=5, swing2LimitDegrees=5, twistLimitDegrees=5}
 {bone1Name=dyn_gps, bone2Name=dyn_gps_base, swing1LimitDegrees=15, swing2LimitDegrees=15, twistLimitDegrees=15}
]
