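# Post-process style AnimBlueprint for the male medium Spy body that copies the attached parent's pose and adds secondary motion and corrective bone adjustments.
# Each update reads thigh rotations relative to the pelvis from the ref pose, driving thigh-volume correctives and GPS prop offsets through range-mapped alphas.
# Gun attachment bones follow weighted spine and clavicle positions with small fixed offsets.
# A final RigidBody node simulates the Spy physics asset in pelvis space with damped component motion.
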AnimBlueprint /Game/Characters/Player/Male/Medium/Bodies/M_MED_Spy/Meshes/M_MED_Spy_AnimBP
target_skeleton=&/M_MED_Spy_Skeleton
graphs [
 AnimGraph {nodes [
   n1 = "Constraint - thigh_volume_l (R)"(ComponentPose=Local To Component(LocalPose=Copy Pose From Mesh() {bUseAttachedParent=true}), Alpha=0.1, ConstraintWeights_0=1) {
    alphaBoolBlend {blendOption=Linear}
    boneToModify {boneName=thigh_volume_l}
    constraintSetup [{transformType=Rotation, targetBone {boneName=thigh_l}}]
    constraintWeights [1]
   }
   n2 = "Transform (Modify) Bone - Bone: thigh_volume_l"(ComponentPose=@n1, Alpha=Get legUp_L() -> (legUp_L:double), Translation="-3,-3,0", Rotation="0,0,0", Scale="1.2,1.2,1.2") {
    rotationMode=BMM_Additive
    rotationSpace=BCS_BoneSpace
    scaleMode=BMM_Additive
    scaleSpace=BCS_BoneSpace
    translationMode=BMM_Additive
    translationSpace=BCS_BoneSpace
    alphaBoolBlend {blendOption=Linear}
    alphaScaleBiasClamp {bMapRange=true, inRange {max=0, min=-100}, outRange {max=0, min=1}}
    boneToModify {boneName=thigh_volume_l}
    scale=<1.2 1.2 1.2>
    translation=<-3 -3>
   }
   n3 = "Constraint - thigh_volume_r (R)"(ComponentPose=@n2, Alpha=0.1, ConstraintWeights_0=1) {
    alphaBoolBlend {blendOption=Linear}
    boneToModify {boneName=thigh_volume_r}
    constraintSetup [{transformType=Rotation, targetBone {boneName=thigh_r}}]
    constraintWeights [1]
   }
   n4 = "Transform (Modify) Bone - Bone: thigh_volume_r"(ComponentPose=@n3, Alpha=Get legUp_R() -> (legUp_R:double), Translation="3,3,0", Rotation="0,0,0", Scale="1.2,1.2,1.2") {
    rotationMode=BMM_Replace
    rotationSpace=BCS_BoneSpace
    scaleMode=BMM_Additive
    scaleSpace=BCS_BoneSpace
    translationMode=BMM_Additive
    translationSpace=BCS_BoneSpace
    alphaBoolBlend {blendOption=Linear}
    alphaScaleBiasClamp {bMapRange=true, inRange {max=0, min=-100}, outRange {max=0, min=1}}
    boneToModify {boneName=thigh_volume_r}
    scale=<1.2 1.2 1.2>
    translation=<3 3>
   }
   n5 = "Constraint - dyn_gun_constr_l (PPPPP)"!(
    ComponentPose=@n4
    Alpha=1
    ConstraintWeights_0=0.0674
    ConstraintWeights_1:float=0.1615
    ConstraintWeights_2:float=0.3756
    ConstraintWeights_3:float=0.2427
    ConstraintWeights_4:float=0.1528
   ) -> (Pose) {
    alphaBoolBlend {blendOption=Linear}
    boneToModify {boneName=dyn_gun_constr_l}
    constraintSetup [
     {targetBone {boneName=spine_01}}
     {targetBone {boneName=spine_02}}
     {targetBone {boneName=spine_03}}
     {targetBone {boneName=spine_04}}
     {targetBone {boneName=clavicle_l}}
    ]
    constraintWeights [1,1,1,1,1]
   }
   n6 = "Constraint - dyn_gun_constr_r (PPPPP)"!(
    ComponentPose=@n5
    Alpha=1
    ConstraintWeights_0=0.1528
    ConstraintWeights_1:float=0.0674
    ConstraintWeights_2:float=0.1615
    ConstraintWeights_3:float=0.3756
    ConstraintWeights_4:float=0.2427
   ) -> (Pose) {
    alphaBoolBlend {blendOption=Linear}
    boneToModify {boneName=dyn_gun_constr_r}
    constraintSetup [
     {targetBone {boneName=clavicle_r}}
     {targetBone {boneName=spine_01}}
     {targetBone {boneName=spine_02}}
     {targetBone {boneName=spine_03}}
     {targetBone {boneName=spine_04}}
    ]
    constraintWeights [1,1,1,1,1]
   }
   n7 = "Transform (Modify) Bone - Bone: dyn_gun_base_l"(ComponentPose=@n6, Alpha=1, Translation="0.5,0,0", Rotation="0,0,0", Scale="1,1,1") {
    rotationMode=BMM_Additive
    rotationSpace=BCS_BoneSpace
    translationMode=BMM_Additive
    translationSpace=BCS_BoneSpace
    alphaBoolBlend {blendOption=Linear}
    boneToModify {boneName=dyn_gun_base_l}
    translation {x=0.5}
   }
   n8 = "Transform (Modify) Bone - Bone: dyn_gun_base_r"(ComponentPose=@n7, Alpha=1, Translation="-0.5,0,0", Rotation="0,0,0", Scale="1,1,1") {
    rotationMode=BMM_Additive
    rotationSpace=BCS_BoneSpace
    translationMode=BMM_Additive
    translationSpace=BCS_BoneSpace
    alphaBoolBlend {blendOption=Linear}
    boneToModify {boneName=dyn_gun_base_r}
    translation {x=-0.5}
   }
   n9 = "Transform (Modify) Bone - Bone: dyn_gps_base"(ComponentPose=@n8, Alpha=1, Translation="0,-1.5,6", Rotation="5,0,-40", Scale="1,1,1") {
    rotationMode=BMM_Additive
    rotationSpace=BCS_BoneSpace
    translationMode=BMM_Additive
    translationSpace=BCS_BoneSpace
    alphaBoolBlend {blendOption=Linear}
    boneToModify {boneName=dyn_gps_base}
    rotation {pitch=5, roll=-40}
    translation {y=-1.5, z=6}
   }
   n10 = "Transform (Modify) Bone - Bone: dyn_gps_base"(ComponentPose=@n9, Alpha=Get legUp_L(), Translation="0,-0.6,3.5", Rotation="10,-20,-15", Scale="1,1,1") {
    rotationMode=BMM_Additive
    rotationSpace=BCS_BoneSpace
    translationMode=BMM_Additive
    translationSpace=BCS_BoneSpace
    alphaBoolBlend {blendOption=Linear}
    alphaScaleBiasClamp {bMapRange=true, inRange {max=0, min=-100}, outRange {max=0, min=1}}
    boneToModify {boneName=dyn_gps_base}
    rotation=<rot 10 -20 -15>
    translation {y=-0.6, z=3.5}
   }
   n11 = "Transform (Modify) Bone - Bone: dyn_gps_base"(ComponentPose=@n10, Alpha=Get legSide_L() -> (legSide_L:double), Translation="0,0,0", Rotation="0,15,-5", Scale="1,1,1") {
    rotationMode=BMM_Additive
    rotationSpace=BCS_BoneSpace
    translationMode=BMM_Additive
    translationSpace=BCS_BoneSpace
    alphaBoolBlend {blendOption=Linear}
    alphaScaleBiasClamp {bMapRange=true, inRange {max=20}}
    boneToModify {boneName=dyn_gps_base}
    rotation {roll=-5, yaw=15}
   }
   n12 = RigidBody(ComponentPose=@n11) {
    bClampLinearTranslationLimitToRefPose=true
    bOverrideWorldGravity=true
    overridePhysicsAsset=&/M_MED_Spy_Physics
    simulationSpace=BaseBoneSpace
    alphaBoolBlend {blendOption=Linear}
    baseBoneRef {boneName=pelvis}
    componentLinearAccScale=<0.45 0.45 0.45>
    componentLinearVelScale=<0.45 0.45 0.45>
    overrideWorldGravity {z=-980}
   }
   Output Pose(Result=Component To Local(ComponentPose=@n12)) {layerGroup=None}
  ]}
 EventGraph {nodes [
   Event BlueprintUpdateAnimation()
   > Sequence()
   > n1 = Sequence()
   n2 = GetOwnerSkeletalMeshComponent()
   n3 = BreakRotator(InRot=BreakTransform(InTransform=GetDeltaTransformFromRefPose(self=@n2, BoneName=thigh_l, BaseName=pelvis)).Rotation)
   > Set legUp_L(legUp_L:double=@n3.Yaw) -> (Output_Get:double)
   > Set legSide_L(legSide_L:double=@n3.Pitch) -> (Output_Get:double)
   n4 = GetDeltaTransformFromRefPose(self=GetOwnerSkeletalMeshComponent(), BoneName=thigh_r, BaseName=pelvis)
   Set legUp_R(execute=@n1.then_1, legUp_R:double=BreakRotator(InRot=BreakTransform(InTransform=@n4).Rotation).Yaw) -> (Output_Get:double)
   TryGetPawnOwner()
  ]}
]
variables [
 legSide_L {type=number}
 legUp_L {type=number}
 legUp_R {type=number}
]
